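# Sound cue for the equip foley of the Figment heavy sniper weapon in Athena.
# A short fixed delay feeds a modulator, which plays a single reused military sniper reload-end wave.
# The cue uses weapon foley sound class and reload attenuation, with reduced volume and its own concurrency override.
# The debug flag is left switched on.
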SoundCue /Figment_Sniper_Heavy_Athena/Sounds/SC_Figment_Sniper_Heavy_Athena_Equip
root=Delay_0
nodes [
 {id=Delay_0, children [Modulator_0], properties {delayMax=0.2, delayMin=0.2}}
 {id=Modulator_0, children [WavePlayer_1], properties {volumeMax=1, volumeMin=1}}
 {id=WavePlayer_1, properties {soundWaveAssetPtr=&/MilitarySniper_ReloadEnd_02}}
]
properties {
 attenuationSettings=&Fort_Weapon_Reload_Att
 bDebug=true
 bOverrideConcurrency=true
 maxDistance=2250
 soundClassObject=&Fort_WeaponFoley
 volumeMultiplier=0.65
}
